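# Niagara effect for the Genie vehicle wheel cosmetic, drawing a circle and orbiting glowing dots with two GPU sprite emitters.
# Circle bursts an effectively permanent, axis-locked sprite offset along local Y, driven by dynamic material parameters.
# Dots continuously spawns small, short-lived sprites with an overbright magenta tint.
# RotateAroundPoint orbits them around the Y axis at that same offset, with randomized radius and speed.
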NiagaraSystem /VehicleCosmetics/Wheels/Genie/PS/PS_Wheel_Genie
dependencies {usedDataInterfaces [], usedDynamicInputs [], usedModules [&SystemState]}
emitters {
 Circle {
  simTarget=GPUComputeSim
  emitterUpdateScript [
   &EmitterState()
   &SpawnBurst_Instantaneous(Use Spawn Probability=false, UseLoopCountLimit=false)
  ]
  particleSpawnScript [
   &InitializeParticle(
    Lifetime=9999999
    Position=<0 20 0>
    Position Mode=::NewEnumerator1
    Position Offset Coordinate Space=::Local
    Sprite Size Mode=::NewEnumerator3
    Uniform Sprite Size=&RandomRangeFloat(Minimum=100, Maximum=100)
    UsePositionOffset=true
    Write Material Random=true
   ) {module=InitializeParticle001}
   &DynamicMaterialParameters(
    Index 0 Param 2=3
    Index 0 Param 4=1
    Module.Index0Float0Write=true
    Module.Index0Float1Write=true
    Module.Index0Float2Write=true
    Module.Index0Float3Write=true
   )
   /CascadeToNiagaraConverter/NiagaraScripts/ModuleScripts/SpriteLockAxis(Lock Axis Type:ENiagara_SpriteLockAxis=::NewEnumerator1, UseAlignmentCoordinateSpace=false, UseFacingCoordinateSpace=false)
  ]
  particleUpdateScript [&ParticleState()]
  rendererClasses [&NiagaraSpriteRendererProperties]
  renderers [{rendererClass=&NiagaraSpriteRendererProperties}]
 }
 Dots {
  simTarget=GPUComputeSim
  emitterUpdateScript [
   &EmitterState(Life Cycle Mode=::NewEnumerator1, Loop Behavior=::NewEnumerator1, Loop Duration Mode=::NewEnumerator1, UseLoopDelay=false)
   &SpawnRate(SpawnRate=10, Use Spawn Probability=false)
  ]
  particleSpawnScript [
   &InitializeParticle(
    AdjustAlpha=true
    AdjustHue=true
    AdjustSaturation=true
    AdjustValue=true
    Alpha Scale Range:Vector2f=<0 1>
    Color=<rgba 1 0.3 1 1>
    Color Mode=::NewEnumerator3
    Hue Shift Range:Vector2f=<1 1>
    Lifetime=2
    Position Mode=::NewEnumerator0
    Saturation Range:Vector2f=<1 1>
    Sprite Rotation Mode=::NewEnumerator2
    Sprite Size Mode=::NewEnumerator1
    Sprite UV Mode=::NewEnumerator3
    Uniform Sprite Size Max=0.8
    Uniform Sprite Size Min=0.3
    Value Range:Vector2f=<1 1>
    Write Material Random=true
   )
   &SpriteLockAxis(Lock Axis Type:ENiagara_SpriteLockAxis=::NewEnumerator1, UseAlignmentCoordinateSpace=false, UseFacingCoordinateSpace=false)
  ]
  particleUpdateScript [
   &ParticleState()
   &ScaleColor(Scale RGB=<200 60 200>, ScaleA=true, ScaleRGB=true)
   &RotateAroundPoint(
    Derive the Rotation Axis=false
    Radius:NiagaraFloat=&RandomRangeFloat(Minimum=24, Maximum=35)
    Rotation Axis:Vector3f=<0 1 0>
    Rotation Center:NiagaraPosition=<0 20 0>
    Rotation Rate:NiagaraFloat=&RandomRangeFloat(Minimum=0, Maximum=-0.6)
   )
  ]
  rendererClasses [&NiagaraSpriteRendererProperties]
  renderers [{rendererClass=&NiagaraSpriteRendererProperties}]
 }
}
summary {
 systemName=PS_Wheel_Genie
 emitters [
  {emitterName=Circle, simTarget=GPUComputeSim, rendererClasses [&NiagaraSpriteRendererProperties]}
  {emitterName=Dots, simTarget=GPUComputeSim, rendererClasses [&NiagaraSpriteRendererProperties]}
 ]
 userVariables []
}
